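# Control Rig for the SleepyUni backpack that makes its five body bones stretch and squash between the dyn_bot and dyn_top bones.
# Setup stores the initial end-to-end distance and builds per-bone linear blend weights, their reverse, and exponents peaking mid-chain.
# Each forward solve position-constrains the bones between both ends.
# It then scales their Y and Z inversely to the stretch ratio, shaped by those exponents and the public volumePreservation variable.
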ControlRigBlueprint /BRCosmetics/Accessories/FORT_Backpacks/Backpack_M_MED_SleepyUni/Meshes/M_MED_SleepyUni_Pack_CtrlRig
graphs [
 New Function_ContainedGraph {nodes [
   n1 = RigVMFunctionEntryNode()
   Divide = &RigVMFunction_MathDoubleDiv(A=@n1.distance, B=@n1.initalDistance)
   Multiply = &RigVMFunction_MathDoubleMul(A=@n1.volume, B=@n1.powExp)
   Power = /Script/RigVM.RigVMFunction_MathDoublePow(A=@Divide.Result, B=@Multiply.Result) {is=struct RigVMUnitNode, template="Power::Execute(in A,in B,out Result)"}
   > RigVMFunctionReturnNode(scale=&RigVMFunction_MathDoubleDiv(A=1, B=@Power.Result).Result)
  ]}
 RigVMFunctionLibrary {nodes [Squash = RigVMCollapseNode(ExecuteContext=())]}
 RigVMFunction_Sequence_1_ContainedGraph {nodes [
   n1 = RigVMFunctionEntryNode()
   RigVMFunction_Sequence = &RigVMFunction_Sequence(ExecuteContext=@n1.ExecuteContext)
   n2 = &RigVMFunction_Sequence(ExecuteContext=@RigVMFunction_Sequence.B)
   RigVMFunctionReturnNode(A=@RigVMFunction_Sequence.A, B=@n2.A, C=@n2.B)
  ]}
 RigVMModel {nodes [
   n1 = &RigUnit_PrepareForExecution()
   &RigUnit_HierarchyImportFromSkeleton(NameSpace=None, bIncludeCurves=False, bIncludeMeshSockets=False, bIncludeVirtualBones=false, ExecutePin=@n1.ExecutePin)
   > RigVMFunction_Sequence = RigVMAggregateNode()
   n2 = &RigUnit_ItemArray(Items=((Type=Bone,Name="kin_body_1"),(Type=Bone,Name="kin_body_2"),(Type=Bone,Name="kin_body_3"),(Type=Bone,Name="kin_body_4"),(Type=Bone,Name="kin_body_5")))
   RigVMVariableNode(Value=@n2.Items, ExecuteContext=@RigVMFunction_Sequence.B) {variable=bendyBones}
   Num = RigVMDispatchNode(Array=RigVMVariableNode() {variable=bendyBones}.Value) {template="DISPATCH_RigVMDispatch_ArrayGetNum(in Array,out Num)"}
   > Set_Num = RigVMDispatchNode(Num=@Num.Num, Array=RigVMVariableNode() {variable=parentConstraintsWeights}.Value) {template="DISPATCH_RigVMDispatch_ArraySetNum(io Array,in Num)"}
   > Set_Num_1 = RigVMDispatchNode(Num=@Num.Num, Array=RigVMVariableNode() {variable=powerExponents}.Value) {template="DISPATCH_RigVMDispatch_ArraySetNum(io Array,in Num)"}
   > RigVMFunction_ForLoopCount = /Script/RigVM.RigVMFunction_ForLoopCount(Count=@Num.Num) {is=struct RigVMUnitNode}
   RigVMFunction_MathIntToDouble_4 = /Script/RigVM.RigVMFunction_MathIntToFloat(Value=/Script/RigVM.RigVMFunction_MathIntSub(B=1, A=@Num.Num) {is=struct RigVMUnitNode, template="Subtract::Execute(in A,in B,out Result)"}.Result) {is=struct RigVMUnitNode, template="Cast::Execute(in Value,out Result)"}
   RigVMFunction_MathIntToDouble_5 = &RigVMFunction_MathIntToFloat(Value=@RigVMFunction_ForLoopCount.Index) {template="Cast::Execute(in Value,out Result)"}
   > Set_At = RigVMDispatchNode(
    Element=&RigVMFunction_MathDoubleDiv(B=@RigVMFunction_MathIntToDouble_4.Result, A=@RigVMFunction_MathIntToDouble_5.Result).Result
    Array=RigVMVariableNode() {variable=parentConstraintsWeights}.Value
    Index=@RigVMFunction_ForLoopCount.Index
   ) {template="DISPATCH_RigVMDispatch_ArraySetAtIndex(io Array,in Index,in Element)"}
   RigVMFunction_MathIntToDouble_6 = &RigVMFunction_MathIntToFloat(Value=/Script/RigVM.RigVMFunction_MathIntAdd(B=1, A=@Num.Num) {is=struct RigVMUnitNode, template="Add::Execute(in A,in B,out Result)"}.Result) {template="Cast::Execute(in Value,out Result)"}
   RigVMFunction_MathIntToDouble_7 = &RigVMFunction_MathIntToFloat(Value=&RigVMFunction_MathIntAdd(B=1, A=@RigVMFunction_ForLoopCount.Index) {template="Add::Execute(in A,in B,out Result)"}.Result) {template="Cast::Execute(in Value,out Result)"}
   n3 = &RigVMFunction_MathDoubleDiv(B=@RigVMFunction_MathIntToDouble_6.Result, A=@RigVMFunction_MathIntToDouble_7.Result)
   > Set_At_1 = RigVMDispatchNode(
    Element=RigVMDispatchNode(
     Condition=/Script/RigVM.RigVMFunction_MathDoubleLessEqual(B=0.5, A=@n3.Result) {is=struct RigVMUnitNode, template="LessEqual::Execute(in A,in B,out Result)"}.Result
     False=&RigVMFunction_MathDoubleSub(A=1, B=@n3.Result).Result
     True=@n3.Result
    ) {template="DISPATCH_RigVMDispatch_If(in Condition,in True,in False,out Result)"}.Result
    Array=RigVMVariableNode() {variable=powerExponents}.Value
    Index=@RigVMFunction_ForLoopCount.Index
   ) {template="DISPATCH_RigVMDispatch_ArraySetAtIndex(io Array,in Index,in Element)"}
   Clone = RigVMDispatchNode(Array=RigVMVariableNode() {variable=parentConstraintsWeights}.Value) {template="DISPATCH_RigVMDispatch_ArrayClone(in Array,out Clone)"}
   n4 = RigVMDispatchNode(Array=@Clone.Clone, ExecuteContext=@RigVMFunction_Sequence.C) {template="DISPATCH_RigVMDispatch_ArrayReverse(io Array)"}
   > RigVMVariableNode(Value=@n4.Array) {variable=parentConstraintsWeightsInverted}
   RigUnit_GetTransform = &RigUnit_GetTransform(Item=(Type=Bone,Name="dyn_bot"), bInitial=true)
   RigVMFunction_MathVectorDistance = &RigVMFunction_MathVectorDistance(
    A=@RigUnit_GetTransform.Transform.Translation
    B=&RigUnit_GetTransform(Item=(Type=Bone,Name="dyn_top"), bInitial=true).Transform.Translation
   )
   RigVMVariableNode(Value=@RigVMFunction_MathVectorDistance.Result, ExecuteContext=@RigVMFunction_Sequence.A) {variable=distanceBetwenTopBot_initial}
   RigUnit_BeginExecution = &RigUnit_BeginExecution()
   RigVMVariableNode(
    ExecuteContext=@RigUnit_BeginExecution.ExecutePin
    Value=&RigVMFunction_MathVectorDistance(
     A=&RigUnit_GetTransform(Item=(Type=Bone,Name="dyn_bot"), bInitial=false).Transform.Translation
     B=&RigUnit_GetTransform(Item=(Type=Bone,Name="dyn_top"), bInitial=false).Transform.Translation
    ).Result
   ) {variable=distanceBetwenTopBot}
   > For_Each = RigVMDispatchNode(Array=RigVMVariableNode() {variable=bendyBones}.Value) {template="DISPATCH_RigVMDispatch_ArrayIterator(in Array,out Element,out Index,out Count,out Ratio)"}
   > &RigUnit_ParentConstraint(
    AdvancedSettings=(InterpolationType=Average,RotationOrderForFilter=XZY)
    Filter=(TranslationFilter=(bX=True,bY=True,bZ=True),RotationFilter=(bX=false,bY=false,bZ=false),ScaleFilter=(bX=false,bY=false,bZ=false))
    Parents=((Item=(Type=Bone,Name="dyn_bot"),Weight=0.25),(Item=(Type=Bone,Name="dyn_top"),Weight=0.75))
    Weight=1
    Parents.1.Weight=RigVMDispatchNode(Index=@For_Each.Index, Array=RigVMVariableNode() {variable=parentConstraintsWeights}.Value) {template="DISPATCH_RigVMDispatch_ArrayGetAtIndex(in Array,in Index,out Element)"}.Element
    Parents.0.Weight=RigVMDispatchNode(Index=@For_Each.Index, Array=RigVMVariableNode() {variable=parentConstraintsWeightsInverted}.Value) {template="DISPATCH_RigVMDispatch_ArrayGetAtIndex(in Array,in Index,out Element)"}.Element
    Child=@For_Each.Element
   )
   > Squash = RigVMFunctionReferenceNode(
    powExp=RigVMDispatchNode(Array=RigVMVariableNode() {variable=powerExponents}.Value, Index=@For_Each.Index) {template="DISPATCH_RigVMDispatch_ArrayGetAtIndex(in Array,in Index,out Element)"}.Element
    distance=RigVMVariableNode() {variable=distanceBetwenTopBot}.Value
    initalDistance=RigVMVariableNode() {variable=distanceBetwenTopBot_initial}.Value
    volume=RigVMVariableNode() {variable=pub volumePreservation}.Value
   )
   > &RigUnit_OffsetTransformForItem(
    OffsetTransform=(Rotation=(X=0,Y=0,Z=0,W=1),Translation=(X=0,Y=0,Z=0),Scale3D=(X=1,Y=1,Z=1))
    Weight=1
    bPropagateToChildren=false
    Item=@For_Each.Element
    OffsetTransform.Scale3D.Y=@Squash.scale
    OffsetTransform.Scale3D.Z=@Squash.scale
   )
  ]}
]
variables [
 bendyBones {type="TArray<FRigElementKey>"}
 distanceBetwenTopBot {type=double, default=0}
 distanceBetwenTopBot_initial {type=double, default=0}
 parentConstraintsWeights {type="TArray<double>"}
 parentConstraintsWeightsInverted {type="TArray<double>"}
 powerExponents {type="TArray<double>"}
 pub volumePreservation {type=double, default=0}
]
