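# Animation Blueprint for the SleepyUni backpack cosmetic, adding procedural secondary motion to the pack and its whale.
# The pose runs through pack offsets, bottom and top springs and a Control Rig, then whale constraint and AnimDynamics in two stiffness variants under a bool blend, and a final tail rotation.
# The event graph computes state-based gravity, velocity-damped spring alphas, spine-driven pack adjustment, tail alpha and a gender-based pack offset.
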AnimBlueprint /BRCosmetics/Accessories/FORT_Backpacks/Backpack_M_MED_SleepyUni/Meshes/M_MED_SleepyUni_Pack_AnimBP
target_skeleton=&/M_MED_SleepyUni_Pack_Skeleton
graphs [
 AnimGraph {nodes [
   n1 = "Constraint - base_whale (PPP)"!(
    ComponentPose=Local To Component(LocalPose=Use cached pose 'controlRig_springsCP'())
    Alpha=1
    ConstraintWeights_0=0.6983
    ConstraintWeights_1:float=0.284
    ConstraintWeights_2:float=0.0177
   ) -> (Pose) {
    alphaBoolBlend {blendOption=Linear}
    boneToModify {boneName=base_whale}
    constraintSetup [{targetBone {boneName=kin_body_5}}, {targetBone {boneName=kin_body_4}}, {targetBone {boneName=kin_body_3}}]
    constraintWeights [1,1,1]
   }
   n2 = Anim Dynamics - Bone: dyn_whale!(ComponentPose=@n1, Alpha=1, GravityOverride:Vector="0,0,-980") -> (Pose) {
    angularDampingOverride=1
    angularSpringConstant=2000
    bAngularSpring=true
    bOverrideAngularDamping=true
    bUseGravityOverride=true
    gravityScale=0.35
    simulationSpace=RootRelative
    alphaBoolBlend {blendOption=Linear}
    boundBone {boneName=dyn_whale}
    componentAppliedLinearAccClamp=<3000 3000 3000>
    componentLinearAccScale=<0.5 0.5 0.5>
    componentLinearVelScale=<0.5 0.5 0.5>
    gravityOverride {z=-980}
    physicsBodyDefinitions [{
      boundBone {boneName=dyn_whale}
      boxExtents=<20 20 20>
      constraintSetup {angularLimitsMax=<0 4 6>, angularLimitsMin=<0 -4 -6>, angularTarget=<1 0 0>}
      localJointOffset=<1 0 0>
     }]
   }
   n3 = Component To Local(ComponentPose=@n2)
   n4 = "Constraint - base_whale (PPP)"(
    ComponentPose=Local To Component(LocalPose=Use cached pose 'controlRig_springsCP'())
    Alpha=1
    ConstraintWeights_0=0.6983
    ConstraintWeights_1=0.284
    ConstraintWeights_2=0.0177
   ) {
    alphaBoolBlend {blendOption=Linear}
    boneToModify {boneName=base_whale}
    constraintSetup [{targetBone {boneName=kin_body_5}}, {targetBone {boneName=kin_body_4}}, {targetBone {boneName=kin_body_3}}]
    constraintWeights [1,1,1]
   }
   n5 = Anim Dynamics - Bone: dyn_whale(ComponentPose=@n4, Alpha=1, GravityOverride="0,0,-980") {
    angularDampingOverride=1
    angularSpringConstant=5000
    bAngularSpring=true
    bOverrideAngularDamping=true
    bUseGravityOverride=true
    gravityScale=0.35
    simulationSpace=RootRelative
    alphaBoolBlend {blendOption=Linear}
    boundBone {boneName=dyn_whale}
    componentAppliedLinearAccClamp=<3000 3000 3000>
    componentLinearAccScale=<0.5 0.5 0.5>
    componentLinearVelScale=<0.5 0.5 0.5>
    gravityOverride {z=-980}
    physicsBodyDefinitions [{
      boundBone {boneName=dyn_whale}
      boxExtents=<20 20 20>
      constraintSetup {angularLimitsMax=<0 4 6>, angularLimitsMin=<0 -4 -6>, angularTarget=<1 0 0>}
      localJointOffset=<4 0 0>
     }]
   }
   n6 = Component To Local(ComponentPose=@n5)
   n7 = "Transform (Modify) Bone - Bone: kin_whaleTail"!(
    ComponentPose=Local To Component(LocalPose=Blend Poses by bool(BlendPose_0=@n3, BlendPose_1=@n6) {blendType=Linear})
    Alpha=1
    Rotation="0,-20,0"
   ) -> (Pose) {
    rotationMode=BMM_Additive
    rotationSpace=BCS_BoneSpace
    alphaBoolBlend {blendOption=Linear}
    boneToModify {boneName=kin_whaleTail}
    rotation {yaw=-20}
   }
   Output Pose(Result=Component To Local(ComponentPose=@n7)) {layerGroup=None}
   n8 = "Transform (Modify) Bone - Bone: pack"!(ComponentPose=Local To Component(LocalPose=Input Pose()), Alpha, Translation="0,-1,0") -> (Pose) {
    rotationMode=BMM_Additive
    rotationSpace=BCS_BoneSpace
    scaleMode=BMM_Additive
    scaleSpace=BCS_BoneSpace
    translationMode=BMM_Additive
    translationSpace=BCS_BoneSpace
    alphaBoolBlend {blendOption=Linear}
    boneToModify {boneName=pack}
    translation {y=-1}
   }
   n9 = "Transform (Modify) Bone - Bone: pack"!(ComponentPose=@n8, Alpha, Translation="0,-1,0", Rotation:Rotator="0,0,0") -> (Pose) {
    rotationMode=BMM_Additive
    rotationSpace=BCS_BoneSpace
    scaleMode=BMM_Additive
    scaleSpace=BCS_BoneSpace
    translationMode=BMM_Additive
    translationSpace=BCS_BoneSpace
    alphaBoolBlend {blendOption=Linear}
    boneToModify {boneName=pack}
    translation {y=-1}
   }
   n10 = Spring controller - Bone: dyn_bot!(ComponentPose:Component Space Pose Link Structure=@n9, Alpha:float=0.05) -> (Pose:Component Space Pose Link Structure) {
    is=title AnimGraphNode_SpringBone
    bLimitDisplacement=true
    bTranslateX=false
    bTranslateY=false
    errorResetThresh=200
    maxDisplacement=100
    springDamping=8
    springStiffness=200
    alphaBoolBlend {blendOption=Linear}
    alphaScaleBiasClamp {bInterpResult=true, interpSpeedDecreasing=15, interpSpeedIncreasing=20}
    springBone {boneName=dyn_bot}
   }
   n11 = Spring controller - Bone: dyn_top!(ComponentPose:Component Space Pose Link Structure=@n10, Alpha:float=0.01) -> (Pose:Component Space Pose Link Structure) {
    is=title AnimGraphNode_SpringBone
    bLimitDisplacement=true
    bTranslateX=false
    bTranslateY=false
    errorResetThresh=200
    maxDisplacement=100
    springDamping=8
    springStiffness=200
    alphaBoolBlend {blendOption=Linear}
    alphaScaleBiasClamp {bInterpResult=true, interpSpeedDecreasing=15, interpSpeedIncreasing=20}
    springBone {boneName=dyn_top}
   }
   n12 = Control Rig!(Source=Component To Local(ComponentPose=@n11), pub volumePreservation:double=2) -> (Pose) {
    alphaBoolBlend {blendOption=Linear}
    controlRigAssetReference {blueprintRigClass=&/M_MED_SleepyUni_Pack_CtrlRig.M_MED_SleepyUni_Pack_CtrlRig_C}
   }
   Save cached pose 'controlRig_springsCP'(Pose=@n12) {cacheName=controlRig_springsCP}
  ]}
 EventGraph {nodes [
   Event BlueprintUpdateAnimation()
   > n1 = Sequence!(execute) -> (then_0, then_1, then_2:Exec)
   > n2 = "Gravity Override State Function (Nativized)"!(
    execute
    CharacterPartAnimInstance:Custom Character Part Anim Instance Object Reference=Self-Reference()
    GravityOverrideParameterSet:Fort Gravity Override Parameter Set Structure=Get Gravity Override Params() -> (Gravity Override Params:Fort Gravity Override Parameter Set Structure)
    RigidBodyAnimNodeParameterSet:Fort Rigid Body Anim Node Parameter Set Structure
    DebugDraw:bool=false
    GameTimeInSeconds:float=GetGameTimeInSeconds()
    DebugJointName:Name=root
    StateLogicLayerName:Name=none
   ) -> (
    then
    OutGravityOverride:Vector
    OutCurrentState:Name=None
    OutGravityOverrideParameters:Fort Gravity Override Parameters Structure
    OutRigidBodyAnimNodeParameters:Fort Rigid Body Anim Node Parameters Structure
   )
   > Set GravityOverride(GravityOverride=@n2)
   n3 = Get PawnVelocityZ()
   n4 = Set springControllerBot alpha(
    execute=@n1.then_1
    springControllerBot alpha:double=Remap Float(Input="Absolute (Float)"(A=@n3), OldRange=(X=900,Y=1250), NewRange=(X=0.05,Y=0), Clamp=true) -> (Output=~)
   ) -> (Output_Get:double)
   > Set springControllerTop alpha(springControllerTop alpha:double=float / float(A=@n4, B=5)) -> (Output_Get:double)
   n5 = GetOwnerSkeletalMeshComponent()
   n6 = BreakRotator(InRot=BreakTransform(InTransform=GetDeltaTransformFromRefPose(self=@n5, BoneName=spine_05, BaseName=spine_01)).Rotation)
   Set backpack_translation(execute=@n1.then_2, backpack_translation=MakeVector(Y=MapRangeClamped(Value=@n6, InRangeA=0, InRangeB=-30, OutRangeB=0.4)))
   > Set backpack_rotate(backpack_rotate=MakeRotator(Roll=MapRangeClamped(Value=@n6.Yaw, InRangeA=0, InRangeB=-30, OutRangeB=5)))
   > Set tailWhaleRot alpha(tailWhaleRot alpha:double=GetAlphaFromReferenceTransformDelta(SkeletalMeshComponent=GetPartSkeletalMeshComponent(), BoneName=base_whale, BaseName=dyn_whale, InputRange=(X=0,Y=-5), TwistAxis="1,0,0")) -> (Output_Get:double)
   n7 = Event OnCharacterCustomizationLoaded()
   n8 = Get Array Item with EFort Custom Part Type(Array:Array of Custom Character Part Object References=@n7.CustomParts) -> (Output:Custom Character Part Object Reference)
   > n9 = Branch(Condition="Equal (Enum)"(
     A:EFortCustomGender Enum=Get GenderPermitted!(self:Custom Character Part Object Reference=@n8) -> (GenderPermitted:EFortCustomGender Enum=Invalid)
     B:EFortCustomGender Enum=Male
    ))
   > Set packOffset_trans(packOffset_trans:Vector="0,-0.8,0") -> (Output_Get:Vector)
   Set packOffset_trans(execute=@n9.else, packOffset_trans="0,-0.6,0")
  ]}
]
variables [
 backpack_rotate {type=Rotator}
 backpack_translation {type=Vector}
 Gravity Override Params {
  type=FortGravityOverrideParameterSet
  default=(Skydive_GoingUp=(GravityOverride=(WindFrequency=(X=1.500000,Y=7.000000,Z=8.000000),WindAmplitude=(X=200.000000,Y=-500.000000,Z=400.000000),GravityOverride=(X=0.000000,Y=-300.000000,Z=1500.000000),JointName="root")),Skydive_FloatingDown=(GravityOverride=(WindFrequency=(X=1.500000,Y=7.000000,Z=8.000000),WindAmplitude=(X=200.000000,Y=-500.000000,Z=400.000000),WindOffset=(X=0.000000,Y=-600.000000,Z=400.000000),GravityOverride=(X=0.000000,Y=-700.000000,Z=600.000000),JointName="root")),Skydive_DivingDown=(GravityOverride=(WindFrequency=(X=1.500000,Y=12.000000,Z=10.000000),WindAmplitude=(X=50.000000,Y=300.000000,Z=100.000000),GravityOverride=(X=-1500.000000,Y=0.000000,Z=0.000000),JointName="pelvis")),Skydive_Parachuting=(GravityOverride=(WindFrequency=(X=1.500000,Y=7.000000,Z=8.000000),WindAmplitude=(X=200.000000,Y=-500.000000,Z=400.000000),GravityOverride=(X=0.000000,Y=-800.000000,Z=-980.000000),JointName="root")),OnGround_Standing=(GravityOverride=(GravityOverride=(X=0.000000,Y=0.000000,Z=-980.000000),JointName="root")),OnGround_Moving=(GravityOverride=(GravityOverride=(X=0.000000,Y=0.000000,Z=-980.000000),JointName="root")),OnGround_Moving_HighRotationRate=(GravityOverride=(GravityOverride=(X=0.000000,Y=0.000000,Z=-980.000000),JointName="root")),OnGround_Crouch_Moving=(GravityOverride=(GravityOverride=(X=0.000000,Y=0.000000,Z=-980.000000),JointName="root")),Emote_or_Melee=(GravityOverride=(GravityOverride=(X=0.000000,Y=0.000000,Z=-980.000000),JointName="root")),Swimming=(GravityOverride=(GravityOverride=(X=-800.000000,Y=0.000000,Z=0.000000),JointName="pelvis")),Front_End=(GravityOverride=(GravityOverride=(X=0.000000,Y=0.000000,Z=-980.000000),JointName="root")),DBNO=(GravityOverride=(GravityOverride=(X=-980.000000,Y=400.000000,Z=0.000000),JointName="pelvis")),DBNO_Carried=(GravityOverride=(GravityOverride=(X=-980.000000,Y=500.000000,Z=-500.000000),JointName="pelvis")),OnGround_Falling=(GravityOverride=(GravityOverride=(X=0.000000,Y=0.000000,Z=-980.000000),JointName="root")),Ziplining=(GravityOverride=(GravityOverride=(X=0.000000,Y=0.000000,Z=-980.000000),JointName="root")),Skydive_Surfing=(GravityOverride=(WindFrequency=(X=1.500000,Y=7.000000,Z=8.000000),WindAmplitude=(X=200.000000,Y=-500.000000,Z=400.000000),GravityOverride=(X=500.000000,Y=-300.000000,Z=-980.000000),JointName="root")),Riding_Standing=(GravityOverride=(GravityOverride=(X=0.000000,Y=-350.000000,Z=-980.000000),JointName="root")),Riding_Moving=(GravityOverride=(GravityOverride=(X=0.000000,Y=-450.000000,Z=-450.000000),JointName="root")),Riding_Moving_Targeting=(GravityOverride=(GravityOverride=(X=0.000000,Y=-200.000000,Z=-980.000000),JointName="root")),Tactical_Sprint=(GravityOverride=(GravityOverride=(X=0.000000,Y=-450.000000,Z=-500.000000),JointName="root")),Rail_Grinding=(GravityOverride=(GravityOverride=(X=0.000000,Y=0.000000,Z=-980.000000),JointName="root")),)
 }
 GravityOverride {type=Vector}
 packOffset_trans {type=Vector}
 springControllerBot alpha {type=number}
 springControllerTop alpha {type=number}
 tailWhaleRot alpha {type=number}
]
